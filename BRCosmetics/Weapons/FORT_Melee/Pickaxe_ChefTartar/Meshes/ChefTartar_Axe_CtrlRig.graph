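# Control Rig for the Chef Tartar pickaxe cosmetic that procedurally waves its flag and grass bones.
# Forward solve runs Chain Harmonics on the flag, then on five grass chains, using slower speeds when in front end and not on pawn.
# Local Chain Dynamics for pole, flag and grass, plus the handle control-to-bone transform, are present but not connected to an execution event.
# A setup event imports the bone hierarchy from the skeleton.
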ControlRigBlueprint /BRCosmetics/Weapons/FORT_Melee/Pickaxe_ChefTartar/Meshes/ChefTartar_Axe_CtrlRig
graphs [
 RigVMFunctionLibrary {links [], nodes []}
 RigVMModel {nodes [
   ItemArray_5 = RigVMDispatchNode(Value=((Type=Bone,Name="dyn_pole_04"),(Type=Bone,Name="dyn_pole_05"),(Type=Bone,Name="dyn_pole_06"),(Type=Bone,Name="dyn_pole_07"))) {template="DISPATCH_RigVMDispatch_Constant(io Value)"}
   Local Chain Dynamics v02_4 = RigVMFunctionReferenceNode(
    Blend End=1
    Blend Start=1
    Blend Twist=1
    Damping=1
    Draw Debug=false
    ExecuteContext=()
    Force=(X=0,Y=0,Z=0)
    Primary Axis=(X=1,Y=0,Z=0)
    Secondary Axis=(X=0,Y=1,Z=0)
    Strength=6
    Bones=@ItemArray_5.Value
   )
   ItemArray = RigVMDispatchNode(Value=((Type=Bone,Name="dyn_flag_01"),(Type=Bone,Name="dyn_flag_02"),(Type=Bone,Name="dyn_flag_03"),(Type=Bone,Name="dyn_flag_04"),(Type=Bone,Name="dyn_flag_05"),(Type=Bone,Name="dyn_flag_06"))) {template="DISPATCH_RigVMDispatch_Constant(io Value)"}
   > Local Chain Dynamics v02 = RigVMFunctionReferenceNode(
    Blend End=0
    Blend Start=1
    Blend Twist=0
    Damping=0.5
    Draw Debug=false
    Force=(X=0,Y=0,Z=0)
    Primary Axis=(X=1,Y=0,Z=0)
    Secondary Axis=(X=0,Y=1,Z=0)
    Strength=5
    Bones=@ItemArray.Value
   )
   ItemArray_1 = RigVMDispatchNode(Value=((Type=Bone,Name="dyn_grass_03"),(Type=Bone,Name="dyn_grass_04"))) {template="DISPATCH_RigVMDispatch_Constant(io Value)"}
   > Local Chain Dynamics v02_1 = RigVMFunctionReferenceNode(
    Blend End=1
    Blend Start=0.8
    Blend Twist=1
    Damping=0.8
    Draw Debug=false
    Force=(X=0,Y=0,Z=0)
    Primary Axis=(X=1,Y=0,Z=0)
    Secondary Axis=(X=0,Y=1,Z=0)
    Strength=5
    Bones=@ItemArray_1.Value
   )
   ItemArray_2 = RigVMDispatchNode(Value=((Type=Bone,Name="dyn_grass_05"),(Type=Bone,Name="dyn_grass_06"))) {template="DISPATCH_RigVMDispatch_Constant(io Value)"}
   > Local Chain Dynamics v02_2 = RigVMFunctionReferenceNode(
    Blend End=1
    Blend Start=0.8
    Blend Twist=1
    Damping=0.8
    Draw Debug=false
    Force=(X=0,Y=0,Z=0)
    Primary Axis=(X=1,Y=0,Z=0)
    Secondary Axis=(X=0,Y=1,Z=0)
    Strength=6
    Bones=@ItemArray_2.Value
   )
   ItemArray_3 = RigVMDispatchNode(Value=((Type=Bone,Name="dyn_grass_01"),(Type=Bone,Name="dyn_grass_02"))) {template="DISPATCH_RigVMDispatch_Constant(io Value)"}
   > Local Chain Dynamics v02_3 = RigVMFunctionReferenceNode(
    Blend End=1
    Blend Start=0.8
    Blend Twist=1
    Damping=1
    Draw Debug=false
    Force=(X=0,Y=0,Z=0)
    Primary Axis=(X=1,Y=0,Z=0)
    Secondary Axis=(X=0,Y=1,Z=0)
    Strength=5
    Bones=@ItemArray_3.Value
   )
   n1 = &RigUnit_PrepareForExecution()
   &RigUnit_HierarchyImportFromSkeleton(NameSpace=None, bIncludeCurves=False, bIncludeMeshSockets=False, bIncludeVirtualBones=True, ExecutePin=@n1.ExecutePin)
   RigUnit_BeginExecution = &RigUnit_BeginExecution()
   n2 = /Script/ControlRig.RigUnit_ChainHarmonicsPerItem(
    ChainRoot=(Type=Bone,Name="base_flag_01")
    DrawWorldOffset=(Rotation=(X=0,Y=0,Z=0,W=1),Translation=(X=0,Y=0,Z=0),Scale3D=(X=1,Y=1,Z=1))
    Pendulum=(bEnabled=False,PendulumStiffness=2,PendulumGravity=(X=0,Y=0,Z=0),PendulumBlend=0.75,PendulumDrag=0.98,PendulumMinimum=0,PendulumMaximum=0.1,PendulumEase=Linear,UnwindAxis=(X=0,Y=1,Z=0),UnwindMinimum=0.2,UnwindMaximum=0.05)
    Reach=(bEnabled=False,ReachTarget=(X=0,Y=0,Z=0),ReachAxis=(X=1,Y=0,Z=0),ReachMinimum=0,ReachMaximum=0,ReachEase=Linear)
    Speed=(X=0,Y=0,Z=6)
    Wave=(bEnabled=True,WaveFrequency=(X=0,Y=0,Z=-85),WaveAmplitude=(X=0,Y=0,Z=1),WaveOffset=(X=0,Y=0,Z=0),WaveNoise=(X=0,Y=0,Z=0),WaveMinimum=0,WaveMaximum=1,WaveEase=Linear)
    WaveCurve=(EditorCurveData=(Keys=((InterpMode=RCIM_Cubic),(Time=1,Value=1))))
    bDrawDebug=True
    ExecutePin=@RigUnit_BeginExecution.ExecutePin
   ) {is=struct RigVMUnitNode}
   RigVMFunction_MathBoolNot = &RigVMFunction_MathBoolNot(Value=RigVMVariableNode() {variable=is on PAWN}.Value)
   RigVMFunction_MathBoolAnd = &RigVMFunction_MathBoolAnd(B=@RigVMFunction_MathBoolNot.Result, A=RigVMVariableNode() {variable=is in FRONT END}.Value)
   RigVMFunction_ControlFlowBranch = &RigVMFunction_ControlFlowBranch(ExecuteContext=@n2.ExecutePin, Condition=@RigVMFunction_MathBoolAnd.Result)
   &RigUnit_ChainHarmonicsPerItem(
    ChainRoot=(Type=Bone,Name="base_grass_07")
    DrawWorldOffset=(Rotation=(X=0,Y=0,Z=0,W=1),Translation=(X=0,Y=0,Z=0),Scale3D=(X=1,Y=1,Z=1))
    Pendulum=(bEnabled=False,PendulumStiffness=2,PendulumGravity=(X=0,Y=0,Z=0),PendulumBlend=0.75,PendulumDrag=0.98,PendulumMinimum=0,PendulumMaximum=0.1,PendulumEase=Linear,UnwindAxis=(X=0,Y=1,Z=0),UnwindMinimum=0.2,UnwindMaximum=0.05)
    Reach=(bEnabled=False,ReachTarget=(X=0,Y=0,Z=0),ReachAxis=(X=1,Y=0,Z=0),ReachMinimum=0,ReachMaximum=0,ReachEase=Linear)
    Speed=(X=0,Y=0,Z=2)
    Wave=(bEnabled=True,WaveFrequency=(X=0,Y=0,Z=50),WaveAmplitude=(X=0,Y=0,Z=0.2),WaveOffset=(X=0,Y=0,Z=0),WaveNoise=(X=0,Y=0,Z=0),WaveMinimum=0,WaveMaximum=1,WaveEase=Linear)
    WaveCurve=(EditorCurveData=(Keys=((InterpMode=RCIM_Cubic),(Time=1,Value=1))))
    bDrawDebug=True
    ExecutePin=@RigVMFunction_ControlFlowBranch.False
   )
   > &RigUnit_ChainHarmonicsPerItem(
    ChainRoot=(Type=Bone,Name="base_grass_08")
    DrawWorldOffset=(Rotation=(X=0,Y=0,Z=0,W=1),Translation=(X=0,Y=0,Z=0),Scale3D=(X=1,Y=1,Z=1))
    Pendulum=(bEnabled=False,PendulumStiffness=2,PendulumGravity=(X=0,Y=0,Z=0),PendulumBlend=0.75,PendulumDrag=0.98,PendulumMinimum=0,PendulumMaximum=0.1,PendulumEase=Linear,UnwindAxis=(X=0,Y=1,Z=0),UnwindMinimum=0.2,UnwindMaximum=0.05)
    Reach=(bEnabled=False,ReachTarget=(X=0,Y=0,Z=0),ReachAxis=(X=1,Y=0,Z=0),ReachMinimum=0,ReachMaximum=0,ReachEase=Linear)
    Speed=(X=0,Y=0,Z=2.5)
    Wave=(bEnabled=True,WaveFrequency=(X=0,Y=0,Z=70),WaveAmplitude=(X=0,Y=0,Z=0.2),WaveOffset=(X=0,Y=0,Z=0),WaveNoise=(X=0,Y=0,Z=0),WaveMinimum=0,WaveMaximum=1,WaveEase=Linear)
    WaveCurve=(EditorCurveData=(Keys=((InterpMode=RCIM_Cubic),(Time=1,Value=1))))
    bDrawDebug=True
   )
   > &RigUnit_ChainHarmonicsPerItem(
    ChainRoot=(Type=Bone,Name="base_grass_01")
    DrawWorldOffset=(Rotation=(X=0,Y=0,Z=0,W=1),Translation=(X=0,Y=0,Z=0),Scale3D=(X=1,Y=1,Z=1))
    Pendulum=(bEnabled=False,PendulumStiffness=2,PendulumGravity=(X=0,Y=0,Z=0),PendulumBlend=0.75,PendulumDrag=0.98,PendulumMinimum=0,PendulumMaximum=0.1,PendulumEase=Linear,UnwindAxis=(X=0,Y=1,Z=0),UnwindMinimum=0.2,UnwindMaximum=0.05)
    Reach=(bEnabled=False,ReachTarget=(X=0,Y=0,Z=0),ReachAxis=(X=1,Y=0,Z=0),ReachMinimum=0,ReachMaximum=0,ReachEase=Linear)
    Speed=(X=0,Y=0,Z=2)
    Wave=(bEnabled=True,WaveFrequency=(X=0,Y=0,Z=70),WaveAmplitude=(X=0,Y=0,Z=0.2),WaveOffset=(X=0,Y=0,Z=0),WaveNoise=(X=0,Y=0,Z=0),WaveMinimum=0,WaveMaximum=1,WaveEase=Linear)
    WaveCurve=(EditorCurveData=(Keys=((InterpMode=RCIM_Cubic),(Time=1,Value=1))))
    bDrawDebug=True
   )
   > &RigUnit_ChainHarmonicsPerItem(
    ChainRoot=(Type=Bone,Name="base_grass_03")
    DrawWorldOffset=(Rotation=(X=0,Y=0,Z=0,W=1),Translation=(X=0,Y=0,Z=0),Scale3D=(X=1,Y=1,Z=1))
    Pendulum=(bEnabled=False,PendulumStiffness=2,PendulumGravity=(X=0,Y=0,Z=0),PendulumBlend=0.75,PendulumDrag=0.98,PendulumMinimum=0,PendulumMaximum=0.1,PendulumEase=Linear,UnwindAxis=(X=0,Y=1,Z=0),UnwindMinimum=0.2,UnwindMaximum=0.05)
    Reach=(bEnabled=False,ReachTarget=(X=0,Y=0,Z=0),ReachAxis=(X=1,Y=0,Z=0),ReachMinimum=0,ReachMaximum=0,ReachEase=Linear)
    Speed=(X=0,Y=0,Z=3)
    Wave=(bEnabled=True,WaveFrequency=(X=0,Y=0,Z=60),WaveAmplitude=(X=0,Y=0,Z=0.2),WaveOffset=(X=0,Y=0,Z=0),WaveNoise=(X=0,Y=0,Z=0),WaveMinimum=0,WaveMaximum=1,WaveEase=Linear)
    WaveCurve=(EditorCurveData=(Keys=((InterpMode=RCIM_Cubic),(Time=1,Value=1))))
    bDrawDebug=True
   )
   > &RigUnit_ChainHarmonicsPerItem(
    ChainRoot=(Type=Bone,Name="base_grass_05")
    DrawWorldOffset=(Rotation=(X=0,Y=0,Z=0,W=1),Translation=(X=0,Y=0,Z=0),Scale3D=(X=1,Y=1,Z=1))
    Pendulum=(bEnabled=False,PendulumStiffness=2,PendulumGravity=(X=0,Y=0,Z=0),PendulumBlend=0.75,PendulumDrag=0.98,PendulumMinimum=0,PendulumMaximum=0.1,PendulumEase=Linear,UnwindAxis=(X=0,Y=1,Z=0),UnwindMinimum=0.2,UnwindMaximum=0.05)
    Reach=(bEnabled=False,ReachTarget=(X=0,Y=0,Z=0),ReachAxis=(X=1,Y=0,Z=0),ReachMinimum=0,ReachMaximum=0,ReachEase=Linear)
    Speed=(X=0,Y=0,Z=2.5)
    Wave=(bEnabled=True,WaveFrequency=(X=0,Y=0,Z=60),WaveAmplitude=(X=0,Y=0,Z=0.2),WaveOffset=(X=0,Y=0,Z=0),WaveNoise=(X=0,Y=0,Z=0),WaveMinimum=0,WaveMaximum=1,WaveEase=Linear)
    WaveCurve=(EditorCurveData=(Keys=((InterpMode=RCIM_Cubic),(Time=1,Value=1))))
    bDrawDebug=True
   )
   &RigUnit_ChainHarmonicsPerItem(
    ChainRoot=(Type=Bone,Name="base_grass_07")
    DrawWorldOffset=(Rotation=(X=0,Y=0,Z=0,W=1),Translation=(X=0,Y=0,Z=0),Scale3D=(X=1,Y=1,Z=1))
    Pendulum=(bEnabled=False,PendulumStiffness=2,PendulumGravity=(X=0,Y=0,Z=0),PendulumBlend=0.75,PendulumDrag=0.98,PendulumMinimum=0,PendulumMaximum=0.1,PendulumEase=Linear,UnwindAxis=(X=0,Y=1,Z=0),UnwindMinimum=0.2,UnwindMaximum=0.05)
    Reach=(bEnabled=False,ReachTarget=(X=0,Y=0,Z=0),ReachAxis=(X=1,Y=0,Z=0),ReachMinimum=0,ReachMaximum=0,ReachEase=Linear)
    Speed=(X=0,Y=0,Z=1.3)
    Wave=(bEnabled=True,WaveFrequency=(X=0,Y=0,Z=50),WaveAmplitude=(X=0,Y=0,Z=0.2),WaveOffset=(X=0,Y=0,Z=0),WaveNoise=(X=0,Y=0,Z=0),WaveMinimum=0,WaveMaximum=1,WaveEase=Linear)
    WaveCurve=(EditorCurveData=(Keys=((InterpMode=RCIM_Cubic),(Time=1,Value=1))))
    bDrawDebug=True
    ExecutePin=@RigVMFunction_ControlFlowBranch.True
   )
   > &RigUnit_ChainHarmonicsPerItem(
    ChainRoot=(Type=Bone,Name="base_grass_08")
    DrawWorldOffset=(Rotation=(X=0,Y=0,Z=0,W=1),Translation=(X=0,Y=0,Z=0),Scale3D=(X=1,Y=1,Z=1))
    Pendulum=(bEnabled=False,PendulumStiffness=2,PendulumGravity=(X=0,Y=0,Z=0),PendulumBlend=0.75,PendulumDrag=0.98,PendulumMinimum=0,PendulumMaximum=0.1,PendulumEase=Linear,UnwindAxis=(X=0,Y=1,Z=0),UnwindMinimum=0.2,UnwindMaximum=0.05)
    Reach=(bEnabled=False,ReachTarget=(X=0,Y=0,Z=0),ReachAxis=(X=1,Y=0,Z=0),ReachMinimum=0,ReachMaximum=0,ReachEase=Linear)
    Speed=(X=0,Y=0,Z=1.7)
    Wave=(bEnabled=True,WaveFrequency=(X=0,Y=0,Z=70),WaveAmplitude=(X=0,Y=0,Z=0.2),WaveOffset=(X=0,Y=0,Z=0),WaveNoise=(X=0,Y=0,Z=0),WaveMinimum=0,WaveMaximum=1,WaveEase=Linear)
    WaveCurve=(EditorCurveData=(Keys=((InterpMode=RCIM_Cubic),(Time=1,Value=1))))
    bDrawDebug=True
   )
   > &RigUnit_ChainHarmonicsPerItem(
    ChainRoot=(Type=Bone,Name="base_grass_01")
    DrawWorldOffset=(Rotation=(X=0,Y=0,Z=0,W=1),Translation=(X=0,Y=0,Z=0),Scale3D=(X=1,Y=1,Z=1))
    Pendulum=(bEnabled=False,PendulumStiffness=2,PendulumGravity=(X=0,Y=0,Z=0),PendulumBlend=0.75,PendulumDrag=0.98,PendulumMinimum=0,PendulumMaximum=0.1,PendulumEase=Linear,UnwindAxis=(X=0,Y=1,Z=0),UnwindMinimum=0.2,UnwindMaximum=0.05)
    Reach=(bEnabled=False,ReachTarget=(X=0,Y=0,Z=0),ReachAxis=(X=1,Y=0,Z=0),ReachMinimum=0,ReachMaximum=0,ReachEase=Linear)
    Speed=(X=0,Y=0,Z=1.5)
    Wave=(bEnabled=True,WaveFrequency=(X=0,Y=0,Z=70),WaveAmplitude=(X=0,Y=0,Z=0.2),WaveOffset=(X=0,Y=0,Z=0),WaveNoise=(X=0,Y=0,Z=0),WaveMinimum=0,WaveMaximum=1,WaveEase=Linear)
    WaveCurve=(EditorCurveData=(Keys=((InterpMode=RCIM_Cubic),(Time=1,Value=1))))
    bDrawDebug=True
   )
   > &RigUnit_ChainHarmonicsPerItem(
    ChainRoot=(Type=Bone,Name="base_grass_03")
    DrawWorldOffset=(Rotation=(X=0,Y=0,Z=0,W=1),Translation=(X=0,Y=0,Z=0),Scale3D=(X=1,Y=1,Z=1))
    Pendulum=(bEnabled=False,PendulumStiffness=2,PendulumGravity=(X=0,Y=0,Z=0),PendulumBlend=0.75,PendulumDrag=0.98,PendulumMinimum=0,PendulumMaximum=0.1,PendulumEase=Linear,UnwindAxis=(X=0,Y=1,Z=0),UnwindMinimum=0.2,UnwindMaximum=0.05)
    Reach=(bEnabled=False,ReachTarget=(X=0,Y=0,Z=0),ReachAxis=(X=1,Y=0,Z=0),ReachMinimum=0,ReachMaximum=0,ReachEase=Linear)
    Speed=(X=0,Y=0,Z=2)
    Wave=(bEnabled=True,WaveFrequency=(X=0,Y=0,Z=60),WaveAmplitude=(X=0,Y=0,Z=0.2),WaveOffset=(X=0,Y=0,Z=0),WaveNoise=(X=0,Y=0,Z=0),WaveMinimum=0,WaveMaximum=1,WaveEase=Linear)
    WaveCurve=(EditorCurveData=(Keys=((InterpMode=RCIM_Cubic),(Time=1,Value=1))))
    bDrawDebug=True
   )
   > &RigUnit_ChainHarmonicsPerItem(
    ChainRoot=(Type=Bone,Name="base_grass_05")
    DrawWorldOffset=(Rotation=(X=0,Y=0,Z=0,W=1),Translation=(X=0,Y=0,Z=0),Scale3D=(X=1,Y=1,Z=1))
    Pendulum=(bEnabled=False,PendulumStiffness=2,PendulumGravity=(X=0,Y=0,Z=0),PendulumBlend=0.75,PendulumDrag=0.98,PendulumMinimum=0,PendulumMaximum=0.1,PendulumEase=Linear,UnwindAxis=(X=0,Y=1,Z=0),UnwindMinimum=0.2,UnwindMaximum=0.05)
    Reach=(bEnabled=False,ReachTarget=(X=0,Y=0,Z=0),ReachAxis=(X=1,Y=0,Z=0),ReachMinimum=0,ReachMaximum=0,ReachEase=Linear)
    Speed=(X=0,Y=0,Z=1.7)
    Wave=(bEnabled=True,WaveFrequency=(X=0,Y=0,Z=60),WaveAmplitude=(X=0,Y=0,Z=0.2),WaveOffset=(X=0,Y=0,Z=0),WaveNoise=(X=0,Y=0,Z=0),WaveMinimum=0,WaveMaximum=1,WaveEase=Linear)
    WaveCurve=(EditorCurveData=(Keys=((InterpMode=RCIM_Cubic),(Time=1,Value=1))))
    bDrawDebug=True
   )
   RigUnit_GetTransform = &RigUnit_GetTransform(Item=(Type=Control,Name="Melee_TwoHanded_Handle_ctrl"), bInitial=False)
   &RigUnit_SetTransform(ExecutePin=(), Item=(Type=Bone,Name="Melee_TwoHanded_Handle"), Weight=1, bInitial=False, Value=@RigUnit_GetTransform.Transform)
  ]}
]
hierarchy [Melee_TwoHanded_Handle_ctrl {controlType=EULER_TRANSFORM, type=Control, loc=<0 0 0>, rot=<0 0 0 1>}]
variables [
 is in FRONT END {type=bool, default=False}
 is on PAWN {type=bool, default=False}
]
