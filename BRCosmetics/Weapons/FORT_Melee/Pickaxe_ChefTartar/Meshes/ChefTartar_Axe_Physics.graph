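# Physics asset for the ChefTartar pickaxe axe mesh, defining physics bodies for a chain of dyn_pole bones.
# Five short capsules cover dyn_pole_03 through dyn_pole_07, with the first one thicker than the rest.
# Four constraints link each bone to its neighbor, locking twist and swing 2 and giving a narrow swing 1 limit, so the chain can bend only in a single plane.
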PhysicsAsset /BRCosmetics/Weapons/FORT_Melee/Pickaxe_ChefTartar/Meshes/ChefTartar_Axe_Physics
bodies {
 dyn_pole_03 [{shapeType=Capsule, length=5, radius=5, shapeName=dyn_pole_03_capsule, center=<5 0 0>, rotation=<rot 90 0 0>}]
 dyn_pole_04 [{shapeType=Capsule, length=5, radius=3, shapeName=dyn_pole_04_capsule, center=<5 0 0>, rotation=<rot 90 0 0>}]
 dyn_pole_05 [{shapeType=Capsule, length=5, radius=3, shapeName=dyn_pole_05_capsule, center=<5 0 0>, rotation=<rot 90 0 0>}]
 dyn_pole_06 [{shapeType=Capsule, length=5, radius=3, shapeName=dyn_pole_06_capsule, center=<5 0 0>, rotation=<rot 90 0 0>}]
 dyn_pole_07 [{shapeType=Capsule, length=5, radius=3, shapeName=dyn_pole_07_capsule, center=<5 0 0>, rotation=<rot 90 0 0>}]
}
constraints [
 {
  bone1Name=dyn_pole_07
  bone2Name=dyn_pole_06
  swing1LimitDegrees=20
  swing2LimitDegrees=1
  swing2Motion=Locked
  twistLimitDegrees=45
  twistMotion=Locked
 }
 {
  bone1Name=dyn_pole_06
  bone2Name=dyn_pole_05
  swing1LimitDegrees=20
  swing2LimitDegrees=1
  swing2Motion=Locked
  twistLimitDegrees=45
  twistMotion=Locked
 }
 {
  bone1Name=dyn_pole_04
  bone2Name=dyn_pole_03
  swing1LimitDegrees=20
  swing2LimitDegrees=1
  swing2Motion=Locked
  twistLimitDegrees=45
  twistMotion=Locked
 }
 {
  bone1Name=dyn_pole_05
  bone2Name=dyn_pole_04
  swing1LimitDegrees=20
  swing2LimitDegrees=1
  swing2Motion=Locked
  twistLimitDegrees=45
  twistMotion=Locked
 }
]
