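# AnimBlueprint for the ChefTartar axe pickaxe, adding secondary motion to its pole, grass and flag bones.
# A Control Rig, fed front-end and on-pawn flags, processes the input pose into a cache that feeds Anim Dynamics grass chains.
# Nested bool blends select between RigidBody on GrassCache_1, grass-plus-pole dynamics with additive bends, or RigidBody followed by flag dynamics.
# Several other pole dynamics and bend chains are left unconnected to the output.
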AnimBlueprint /BRCosmetics/Weapons/FORT_Melee/Pickaxe_ChefTartar/Meshes/ChefTartar_Axe_AnimBP
target_skeleton=&/ChefTartar_Axe_Skeleton
graphs [
 AnimGraph {nodes [
   n1 = Anim Dynamics - Bone: dyn_pole_04(Alpha=1, ExternalForce="0,0,0") {
    angularDampingOverride=1
    angularSpringConstant=60000
    bAngularSpring=true
    bOverrideAngularDamping=true
    bUseGravityOverride=true
    gravityScale=0.35
    simulationSpace=RootRelative
    boundBone {boneName=dyn_pole_04}
    componentLinearAccScale=<1 1 1>
    componentLinearVelScale=<1 1 1>
    gravityOverride {z=-980}
    physicsBodyDefinitions [{
      boundBone {boneName=dyn_pole_04}
      boxExtents=<30 40 40>
      constraintSetup {angularLimitsMax=<0 0 7>, angularLimitsMin=<0 0 -7>, angularTarget=<1 0 0>}
      localJointOffset=<4 0 0>
     }]
   }
   n2 = Anim Dynamics - Bone: dyn_pole_05(ComponentPose=@n1, Alpha=1, ExternalForce="0,0,0") {
    angularDampingOverride=1
    angularSpringConstant=60000
    bAngularSpring=true
    bOverrideAngularDamping=true
    bUseGravityOverride=true
    gravityScale=0.35
    simulationSpace=RootRelative
    boundBone {boneName=dyn_pole_05}
    componentLinearAccScale=<1 1 1>
    componentLinearVelScale=<1 1 1>
    gravityOverride {z=-980}
    physicsBodyDefinitions [{
      boundBone {boneName=dyn_pole_05}
      boxExtents=<30 40 40>
      constraintSetup {angularLimitsMax=<0 0 7>, angularLimitsMin=<0 0 -7>, angularTarget=<1 0 0>}
      localJointOffset=<4 0 0>
     }]
   }
   n3 = Anim Dynamics - Bone: dyn_pole_06(ComponentPose=@n2, Alpha=1, ExternalForce="0,0,0") {
    angularDampingOverride=1
    angularSpringConstant=60000
    bAngularSpring=true
    bOverrideAngularDamping=true
    bUseGravityOverride=true
    gravityScale=0.35
    simulationSpace=RootRelative
    boundBone {boneName=dyn_pole_06}
    componentLinearAccScale=<1 1 1>
    componentLinearVelScale=<1 1 1>
    gravityOverride {z=-980}
    physicsBodyDefinitions [{
      boundBone {boneName=dyn_pole_06}
      boxExtents=<30 40 40>
      constraintSetup {angularLimitsMax=<0 0 7>, angularLimitsMin=<0 0 -7>, angularTarget=<1 0 0>}
      localJointOffset=<4 0 0>
     }]
   }
   Anim Dynamics - Bone: dyn_pole_07(ComponentPose=@n3, Alpha=1, ExternalForce="0,0,0") {
    angularDampingOverride=1
    angularSpringConstant=60000
    bAngularSpring=true
    bOverrideAngularDamping=true
    bUseGravityOverride=true
    gravityScale=0.35
    simulationSpace=RootRelative
    boundBone {boneName=dyn_pole_07}
    componentLinearAccScale=<1 1 1>
    componentLinearVelScale=<1 1 1>
    gravityOverride {z=-980}
    physicsBodyDefinitions [{
      boundBone {boneName=dyn_pole_07}
      boxExtents=<30 40 40>
      constraintSetup {angularLimitsMax=<0 0 7>, angularLimitsMin=<0 0 -7>, angularTarget=<1 0 0>}
      localJointOffset=<4 0 0>
     }]
   }
   n4 = Anim Dynamics - Bone: dyn_pole_04(ComponentPose=Local To Component(), Alpha=1, ExternalForce="0,0,0") {
    angularDampingOverride=1
    angularSpringConstant=20000
    bAngularSpring=true
    bOverrideAngularDamping=true
    bUseGravityOverride=true
    gravityScale=0.35
    simulationSpace=RootRelative
    boundBone {boneName=dyn_pole_04}
    componentLinearAccScale=<1 1 1>
    componentLinearVelScale=<1 1 1>
    gravityOverride {z=-1500}
    physicsBodyDefinitions [{
      boundBone {boneName=dyn_pole_04}
      boxExtents=<20 30 30>
      constraintSetup {angularLimitsMax=<0 0 2>, angularLimitsMin=<0 0 -7>, angularTarget=<1 0 0>}
      localJointOffset=<4 0 0>
     }]
   }
   n5 = Anim Dynamics - Bone: dyn_pole_05(ComponentPose=@n4, Alpha=1, ExternalForce="0,0,0") {
    angularDampingOverride=1
    angularSpringConstant=20000
    bAngularSpring=true
    bOverrideAngularDamping=true
    bUseGravityOverride=true
    gravityScale=0.35
    simulationSpace=RootRelative
    boundBone {boneName=dyn_pole_05}
    componentLinearAccScale=<1 1 1>
    componentLinearVelScale=<1 1 1>
    gravityOverride {z=-1500}
    physicsBodyDefinitions [{
      boundBone {boneName=dyn_pole_05}
      boxExtents=<20 30 30>
      constraintSetup {angularLimitsMax=<0 0 2>, angularLimitsMin=<0 0 -7>, angularTarget=<1 0 0>}
      localJointOffset=<4 0 0>
     }]
   }
   n6 = Anim Dynamics - Bone: dyn_pole_06(ComponentPose=@n5, Alpha=1, ExternalForce="0,0,0") {
    angularDampingOverride=1
    angularSpringConstant=20000
    bAngularSpring=true
    bOverrideAngularDamping=true
    bUseGravityOverride=true
    gravityScale=0.35
    simulationSpace=RootRelative
    boundBone {boneName=dyn_pole_06}
    componentLinearAccScale=<1 1 1>
    componentLinearVelScale=<1 1 1>
    gravityOverride {z=-1500}
    physicsBodyDefinitions [{
      boundBone {boneName=dyn_pole_06}
      boxExtents=<20 30 30>
      constraintSetup {angularLimitsMax=<0 0 2>, angularLimitsMin=<0 0 -7>, angularTarget=<1 0 0>}
      localJointOffset=<4 0 0>
     }]
   }
   n7 = Anim Dynamics - Bone: dyn_pole_07(ComponentPose=@n6, Alpha=1, ExternalForce="0,0,0") {
    angularDampingOverride=1
    angularSpringConstant=20000
    bAngularSpring=true
    bOverrideAngularDamping=true
    bUseGravityOverride=true
    gravityScale=0.35
    simulationSpace=RootRelative
    boundBone {boneName=dyn_pole_07}
    componentLinearAccScale=<1 1 1>
    componentLinearVelScale=<1 1 1>
    gravityOverride {z=-1500}
    physicsBodyDefinitions [{
      boundBone {boneName=dyn_pole_07}
      boxExtents=<20 30 30>
      constraintSetup {angularLimitsMax=<0 0 2>, angularLimitsMin=<0 0 -7>, angularTarget=<1 0 0>}
      localJointOffset=<4 0 0>
     }]
   }
   n8 = "Transform (Modify) Bone - Bone: dyn_pole_04"!(ComponentPose=@n7, bAlphaBoolEnabled:bool=True, Translation="0,0,0", Scale="1,1,1") -> (Pose) {
    alphaInputType=Bool
    rotationMode=BMM_Additive
    rotationSpace=BCS_BoneSpace
    translationMode=BMM_Additive
    translationSpace=BCS_BoneSpace
    alphaBoolBlend {blendInTime=0.05, blendOutTime=0.05}
    boneToModify {boneName=dyn_pole_04}
    rotation {yaw=6}
   }
   n9 = "Transform (Modify) Bone - Bone: dyn_pole_05"!(ComponentPose=@n8, bAlphaBoolEnabled:bool=True, Translation="0,0,0", Scale="1,1,1") -> (Pose) {
    alphaInputType=Bool
    rotationMode=BMM_Additive
    rotationSpace=BCS_BoneSpace
    translationMode=BMM_Additive
    translationSpace=BCS_BoneSpace
    alphaBoolBlend {blendInTime=0.05, blendOutTime=0.05}
    boneToModify {boneName=dyn_pole_05}
    rotation {yaw=6}
   }
   n10 = "Transform (Modify) Bone - Bone: dyn_pole_06"!(ComponentPose=@n9, bAlphaBoolEnabled:bool=True, Translation="0,0,0", Scale="1,1,1") -> (Pose) {
    alphaInputType=Bool
    rotationMode=BMM_Additive
    rotationSpace=BCS_BoneSpace
    translationMode=BMM_Additive
    translationSpace=BCS_BoneSpace
    alphaBoolBlend {blendInTime=0.05, blendOutTime=0.05}
    boneToModify {boneName=dyn_pole_06}
    rotation {yaw=6}
   }
   n11 = "Transform (Modify) Bone - Bone: dyn_pole_07"!(ComponentPose=@n10, bAlphaBoolEnabled:bool=True, Translation="0,0,0", Scale="1,1,1") -> (Pose) {
    alphaInputType=Bool
    rotationMode=BMM_Additive
    rotationSpace=BCS_BoneSpace
    translationMode=BMM_Additive
    translationSpace=BCS_BoneSpace
    alphaBoolBlend {blendInTime=0.05, blendOutTime=0.05}
    boneToModify {boneName=dyn_pole_07}
    rotation {yaw=6}
   }
   Component To Local(ComponentPose=@n11)
   n12 = Anim Dynamics - Bone: dyn_pole_01(Alpha=0.2, ExternalForce="0,0,0") {
    alpha=0.2
    angularSpringConstant=10000
    bAngularSpring=true
    linearSpringConstant=1000
    simulationSpace=RootRelative
    boundBone {boneName=dyn_pole_01}
    gravityOverride {z=-1500}
    physicsBodyDefinitions [{
      boundBone {boneName=dyn_pole_01}
      boxExtents=<5 5 5>
      constraintSetup {angularLimitsMax=<0 0 -9>, angularLimitsMin=<0 0 -30>, angularTarget=<0 0 0>}
      localJointOffset=<0.1 0 0>
     }]
    simSpaceSettings {simSpaceAngularAlpha=0.4}
   }
   n13 = Anim Dynamics - Bone: dyn_pole_02(ComponentPose=@n12, Alpha=0.2, ExternalForce="0,0,0") {
    alpha=0.2
    angularSpringConstant=10000
    bAngularSpring=true
    linearSpringConstant=1000
    simulationSpace=RootRelative
    boundBone {boneName=dyn_pole_02}
    gravityOverride {z=-1500}
    physicsBodyDefinitions [{
      boundBone {boneName=dyn_pole_02}
      boxExtents=<5 5 5>
      constraintSetup {angularLimitsMax=<0 0 -9>, angularLimitsMin=<0 0 -30>, angularTarget=<0 0 0>}
      localJointOffset=<0.1 0 0>
     }]
    simSpaceSettings {simSpaceAngularAlpha=0.4}
   }
   n14 = Anim Dynamics - Bone: dyn_pole_03(ComponentPose=@n13, Alpha=0.2, ExternalForce="0,0,0") {
    alpha=0.2
    angularSpringConstant=10000
    bAngularSpring=true
    linearSpringConstant=1000
    simulationSpace=RootRelative
    boundBone {boneName=dyn_pole_03}
    gravityOverride {z=-1500}
    physicsBodyDefinitions [{
      boundBone {boneName=dyn_pole_03}
      boxExtents=<5 5 5>
      constraintSetup {angularLimitsMax=<0 0 -9>, angularLimitsMin=<0 0 -30>, angularTarget=<0 0 0>}
      localJointOffset=<0.1 0 0>
     }]
    simSpaceSettings {simSpaceAngularAlpha=0.4}
   }
   n15 = Anim Dynamics - Bone: dyn_pole_04(ComponentPose=@n14, Alpha=0.2, ExternalForce="0,0,0") {
    alpha=0.2
    angularSpringConstant=10000
    bAngularSpring=true
    linearSpringConstant=1000
    simulationSpace=RootRelative
    boundBone {boneName=dyn_pole_04}
    gravityOverride {z=-1500}
    physicsBodyDefinitions [{
      boundBone {boneName=dyn_pole_04}
      boxExtents=<5 5 5>
      constraintSetup {angularLimitsMax=<0 0 -6>, angularLimitsMin=<0 0 -20>, angularTarget=<0 0 0>}
      localJointOffset=<0.1 0 0>
     }]
    simSpaceSettings {simSpaceAngularAlpha=0.4}
   }
   n16 = Anim Dynamics - Bone: dyn_pole_05(ComponentPose=@n15, Alpha=0.2, ExternalForce="0,0,0") {
    alpha=0.2
    angularSpringConstant=10000
    bAngularSpring=true
    linearSpringConstant=1000
    simulationSpace=RootRelative
    boundBone {boneName=dyn_pole_05}
    gravityOverride {z=-1500}
    physicsBodyDefinitions [{
      boundBone {boneName=dyn_pole_05}
      boxExtents=<5 5 5>
      constraintSetup {angularLimitsMax=<0 0 -6>, angularLimitsMin=<0 0 -20>, angularTarget=<0 0 0>}
      localJointOffset=<0.1 0 0>
     }]
    simSpaceSettings {simSpaceAngularAlpha=0.4}
   }
   n17 = Anim Dynamics - Bone: dyn_pole_06(ComponentPose=@n16, Alpha=0.2, ExternalForce="0,0,0") {
    alpha=0.2
    angularSpringConstant=10000
    bAngularSpring=true
    linearSpringConstant=1000
    simulationSpace=RootRelative
    boundBone {boneName=dyn_pole_06}
    gravityOverride {z=-1500}
    physicsBodyDefinitions [{
      boundBone {boneName=dyn_pole_06}
      boxExtents=<5 5 5>
      constraintSetup {angularLimitsMax=<0 0 -6>, angularLimitsMin=<0 0 -20>, angularTarget=<0 0 0>}
      localJointOffset=<0.1 0 0>
     }]
    simSpaceSettings {simSpaceAngularAlpha=0.4}
   }
   n18 = Anim Dynamics - Bone: dyn_pole_07(ComponentPose=@n17, Alpha=0.2, ExternalForce="0,0,0") {
    alpha=0.2
    angularSpringConstant=10000
    bAngularSpring=true
    linearSpringConstant=1000
    simulationSpace=RootRelative
    boundBone {boneName=dyn_pole_07}
    gravityOverride {z=-1500}
    physicsBodyDefinitions [{
      boundBone {boneName=dyn_pole_07}
      boxExtents=<5 5 5>
      constraintSetup {angularLimitsMax=<0 0 -6>, angularLimitsMin=<0 0 -20>, angularTarget=<0 0 0>}
      localJointOffset=<0.1 0 0>
     }]
    simSpaceSettings {simSpaceAngularAlpha=0.4}
   }
   n19 = "Transform (Modify) Bone - Bone: dyn_pole_01"!(ComponentPose=@n18, bAlphaBoolEnabled:bool=True, Translation="0,0,0", Scale="1,1,1") -> (Pose) {
    alphaInputType=Bool
    rotationMode=BMM_Additive
    rotationSpace=BCS_BoneSpace
    translationMode=BMM_Additive
    translationSpace=BCS_BoneSpace
    alphaBoolBlend {blendInTime=0.3, blendOutTime=0.3}
    boneToModify {boneName=dyn_pole_01}
    rotation {yaw=-6}
   }
   n20 = "Transform (Modify) Bone - Bone: dyn_pole_02"!(ComponentPose=@n19, bAlphaBoolEnabled:bool=True, Translation="0,0,0", Scale="1,1,1") -> (Pose) {
    alphaInputType=Bool
    rotationMode=BMM_Additive
    rotationSpace=BCS_BoneSpace
    translationMode=BMM_Additive
    translationSpace=BCS_BoneSpace
    alphaBoolBlend {blendInTime=0.3, blendOutTime=0.3}
    boneToModify {boneName=dyn_pole_02}
    rotation {yaw=-6}
   }
   n21 = "Transform (Modify) Bone - Bone: dyn_pole_03"!(ComponentPose=@n20, bAlphaBoolEnabled:bool=True, Translation="0,0,0", Scale="1,1,1") -> (Pose) {
    alphaInputType=Bool
    rotationMode=BMM_Additive
    rotationSpace=BCS_BoneSpace
    translationMode=BMM_Additive
    translationSpace=BCS_BoneSpace
    alphaBoolBlend {blendInTime=0.3, blendOutTime=0.3}
    boneToModify {boneName=dyn_pole_03}
    rotation {yaw=-6}
   }
   n22 = "Transform (Modify) Bone - Bone: dyn_pole_04"(ComponentPose=@n21, bAlphaBoolEnabled=True, Translation="0,0,0", Scale="1,1,1") {
    alphaInputType=Bool
    rotationMode=BMM_Additive
    rotationSpace=BCS_BoneSpace
    translationMode=BMM_Additive
    translationSpace=BCS_BoneSpace
    alphaBoolBlend {blendInTime=0.3, blendOutTime=0.3}
    boneToModify {boneName=dyn_pole_04}
    rotation {yaw=-6}
   }
   n23 = "Transform (Modify) Bone - Bone: dyn_pole_05"(ComponentPose=@n22, bAlphaBoolEnabled=True, Translation="0,0,0", Scale="1,1,1") {
    alphaInputType=Bool
    rotationMode=BMM_Additive
    rotationSpace=BCS_BoneSpace
    translationMode=BMM_Additive
    translationSpace=BCS_BoneSpace
    alphaBoolBlend {blendInTime=0.3, blendOutTime=0.3}
    boneToModify {boneName=dyn_pole_05}
    rotation {yaw=-6}
   }
   n24 = "Transform (Modify) Bone - Bone: dyn_pole_06"(ComponentPose=@n23, bAlphaBoolEnabled=True, Translation="0,0,0", Scale="1,1,1") {
    alphaInputType=Bool
    rotationMode=BMM_Additive
    rotationSpace=BCS_BoneSpace
    translationMode=BMM_Additive
    translationSpace=BCS_BoneSpace
    alphaBoolBlend {blendInTime=0.3, blendOutTime=0.3}
    boneToModify {boneName=dyn_pole_06}
    rotation {yaw=-6}
   }
   "Transform (Modify) Bone - Bone: dyn_pole_07"(ComponentPose=@n24, bAlphaBoolEnabled=True, Translation="0,0,0", Scale="1,1,1") {
    alphaInputType=Bool
    rotationMode=BMM_Additive
    rotationSpace=BCS_BoneSpace
    translationMode=BMM_Additive
    translationSpace=BCS_BoneSpace
    alphaBoolBlend {blendInTime=0.3, blendOutTime=0.3}
    boneToModify {boneName=dyn_pole_07}
    rotation {yaw=-6}
   }
   n25 = "Transform (Modify) Bone - Bone: dyn_pole_04"(bAlphaBoolEnabled=True, Translation="0,0,0", Scale="1,1,1") {
    alphaInputType=Bool
    rotationMode=BMM_Additive
    rotationSpace=BCS_BoneSpace
    translationMode=BMM_Additive
    translationSpace=BCS_BoneSpace
    alphaBoolBlend {blendInTime=0.3, blendOutTime=0.3}
    boneToModify {boneName=dyn_pole_04}
    rotation {yaw=6}
   }
   n26 = "Transform (Modify) Bone - Bone: dyn_pole_05"(ComponentPose=@n25, bAlphaBoolEnabled=True, Translation="0,0,0", Scale="1,1,1") {
    alphaInputType=Bool
    rotationMode=BMM_Additive
    rotationSpace=BCS_BoneSpace
    translationMode=BMM_Additive
    translationSpace=BCS_BoneSpace
    alphaBoolBlend {blendInTime=0.3, blendOutTime=0.3}
    boneToModify {boneName=dyn_pole_05}
    rotation {yaw=6}
   }
   n27 = "Transform (Modify) Bone - Bone: dyn_pole_06"(ComponentPose=@n26, bAlphaBoolEnabled=True, Translation="0,0,0", Scale="1,1,1") {
    alphaInputType=Bool
    rotationMode=BMM_Additive
    rotationSpace=BCS_BoneSpace
    translationMode=BMM_Additive
    translationSpace=BCS_BoneSpace
    alphaBoolBlend {blendInTime=0.3, blendOutTime=0.3}
    boneToModify {boneName=dyn_pole_06}
    rotation {yaw=6}
   }
   "Transform (Modify) Bone - Bone: dyn_pole_07"(ComponentPose=@n27, bAlphaBoolEnabled=True, Translation="0,0,0", Scale="1,1,1") {
    alphaInputType=Bool
    rotationMode=BMM_Additive
    rotationSpace=BCS_BoneSpace
    translationMode=BMM_Additive
    translationSpace=BCS_BoneSpace
    alphaBoolBlend {blendInTime=0.3, blendOutTime=0.3}
    boneToModify {boneName=dyn_pole_07}
    rotation {yaw=6}
   }
   n28 = RigidBody!(
    ComponentPose=Local To Component(LocalPose=Use cached pose 'GrassCache_1'())
    bAlphaBoolEnabled:bool
    OverrideWorldGravity:Vector="0,0,0"
    ExternalForce
    SimSpaceSettings:Sim Space Settings Structure=(WorldAlpha=0,VelocityScaleZ=1,DampingAlpha=1,MaxLinearVelocity=10000,MaxAngularVelocity=10000,MaxLinearAcceleration=10000,MaxAngularAcceleration=10000,ExternalLinearDragV=(X=0,Y=0,Z=0),ExternalLinearVelocity=(X=0,Y=0,Z=0),ExternalAngularVelocity=(X=0,Y=0,Z=0))
   ) -> (Pose) {
    alphaInputType=Bool
    bOverrideWorldGravity=true
    overridePhysicsAsset=&/ChefTartar_Axe_Physics
   }
   n29 = Component To Local(ComponentPose=@n28)
   n30 = Anim Dynamics - Bone: base_grass_01!(ComponentPose=Local To Component(LocalPose=Use cached pose 'input_pose_cache'()), ExternalForce="0,0,0") -> (Pose) {
    alpha=0.3
    angularDampingOverride=1
    angularSpringConstant=20000
    bAngularSpring=true
    bOverrideAngularDamping=true
    bUseGravityOverride=true
    gravityScale=0.35
    simulationSpace=World
    boundBone {boneName=base_grass_01}
    componentLinearAccScale=<1 1 1>
    componentLinearVelScale=<1 1 1>
    gravityOverride {z=-980}
    physicsBodyDefinitions [{
      boundBone {boneName=base_grass_01}
      boxExtents=<20 30 30>
      constraintSetup {angularLimitsMax=<15 15 15>, angularLimitsMin=<-15 -15 -15>, angularTarget=<1 0 0>}
      localJointOffset=<2 0 0>
     }]
   }
   n31 = Anim Dynamics - Bone: dyn_grass_02!(ComponentPose=@n30, ExternalForce="0,0,0") -> (Pose) {
    alpha=0.3
    angularDampingOverride=1
    angularSpringConstant=20000
    bAngularSpring=true
    bOverrideAngularDamping=true
    bUseGravityOverride=true
    gravityScale=0.35
    simulationSpace=World
    boundBone {boneName=dyn_grass_02}
    componentLinearAccScale=<1 1 1>
    componentLinearVelScale=<1 1 1>
    gravityOverride {z=-980}
    physicsBodyDefinitions [{
      boundBone {boneName=dyn_grass_02}
      boxExtents=<20 30 30>
      constraintSetup {angularLimitsMax=<15 15 15>, angularLimitsMin=<-15 -15 -15>, angularTarget=<1 0 0>}
      localJointOffset=<2 0 0>
     }]
   }
   n32 = Anim Dynamics - Bone: base_grass_03!(ComponentPose=@n31, ExternalForce="0,0,0") -> (Pose) {
    alpha=0.3
    angularDampingOverride=1
    angularSpringConstant=20000
    bAngularSpring=true
    bOverrideAngularDamping=true
    bUseGravityOverride=true
    gravityScale=0.35
    simulationSpace=World
    boundBone {boneName=base_grass_03}
    componentLinearAccScale=<1 1 1>
    componentLinearVelScale=<1 1 1>
    gravityOverride {z=-980}
    physicsBodyDefinitions [{
      boundBone {boneName=base_grass_03}
      boxExtents=<20 30 30>
      constraintSetup {angularLimitsMax=<15 15 15>, angularLimitsMin=<-15 -15 -15>, angularTarget=<1 0 0>}
      localJointOffset=<2 0 0>
     }]
   }
   n33 = Anim Dynamics - Bone: dyn_grass_04!(ComponentPose=@n32, ExternalForce="0,0,0") -> (Pose) {
    alpha=0.3
    angularDampingOverride=1
    angularSpringConstant=20000
    bAngularSpring=true
    bOverrideAngularDamping=true
    bUseGravityOverride=true
    gravityScale=0.35
    simulationSpace=World
    boundBone {boneName=dyn_grass_04}
    componentLinearAccScale=<1 1 1>
    componentLinearVelScale=<1 1 1>
    gravityOverride {z=-980}
    physicsBodyDefinitions [{
      boundBone {boneName=dyn_grass_04}
      boxExtents=<20 30 30>
      constraintSetup {angularLimitsMax=<15 15 15>, angularLimitsMin=<-15 -15 -15>, angularTarget=<1 0 0>}
      localJointOffset=<2 0 0>
     }]
   }
   n34 = Anim Dynamics - Bone: base_grass_05!(ComponentPose=@n33, ExternalForce="0,0,0") -> (Pose) {
    alpha=0.3
    angularDampingOverride=1
    angularSpringConstant=20000
    bAngularSpring=true
    bOverrideAngularDamping=true
    bUseGravityOverride=true
    gravityScale=0.35
    simulationSpace=World
    boundBone {boneName=base_grass_05}
    componentLinearAccScale=<1 1 1>
    componentLinearVelScale=<1 1 1>
    gravityOverride {z=-980}
    physicsBodyDefinitions [{
      boundBone {boneName=base_grass_05}
      boxExtents=<20 30 30>
      constraintSetup {angularLimitsMax=<15 15 15>, angularLimitsMin=<-15 -15 -15>, angularTarget=<1 0 0>}
      localJointOffset=<2 0 0>
     }]
   }
   n35 = Anim Dynamics - Bone: dyn_grass_06!(ComponentPose=@n34, ExternalForce="0,0,0") -> (Pose) {
    alpha=0.3
    angularDampingOverride=1
    angularSpringConstant=20000
    bAngularSpring=true
    bOverrideAngularDamping=true
    bUseGravityOverride=true
    gravityScale=0.35
    simulationSpace=World
    boundBone {boneName=dyn_grass_06}
    componentLinearAccScale=<1 1 1>
    componentLinearVelScale=<1 1 1>
    gravityOverride {z=-980}
    physicsBodyDefinitions [{
      boundBone {boneName=dyn_grass_06}
      boxExtents=<20 30 30>
      constraintSetup {angularLimitsMax=<15 15 15>, angularLimitsMin=<-15 -15 -15>, angularTarget=<1 0 0>}
      localJointOffset=<2 0 0>
     }]
   }
   n36 = Anim Dynamics - Bone: base_grass_07!(ComponentPose=@n35, ExternalForce="0,0,0") -> (Pose) {
    alpha=0.3
    angularDampingOverride=1
    angularSpringConstant=20000
    bAngularSpring=true
    bOverrideAngularDamping=true
    bUseGravityOverride=true
    gravityScale=0.35
    simulationSpace=World
    boundBone {boneName=base_grass_07}
    componentLinearAccScale=<1 1 1>
    componentLinearVelScale=<1 1 1>
    gravityOverride {z=-980}
    physicsBodyDefinitions [{
      boundBone {boneName=base_grass_07}
      boxExtents=<20 30 30>
      constraintSetup {angularLimitsMax=<15 15 15>, angularLimitsMin=<-15 -15 -15>, angularTarget=<1 0 0>}
      localJointOffset=<2 0 0>
     }]
   }
   n37 = Anim Dynamics - Bone: base_grass_08!(ComponentPose=@n36, ExternalForce="0,0,0") -> (Pose) {
    alpha=0.3
    angularDampingOverride=1
    angularSpringConstant=20000
    bAngularSpring=true
    bOverrideAngularDamping=true
    bUseGravityOverride=true
    gravityScale=0.35
    simulationSpace=World
    boundBone {boneName=base_grass_08}
    componentLinearAccScale=<1 1 1>
    componentLinearVelScale=<1 1 1>
    gravityOverride {z=-980}
    physicsBodyDefinitions [{
      boundBone {boneName=base_grass_08}
      boxExtents=<20 30 30>
      constraintSetup {angularLimitsMax=<15 15 15>, angularLimitsMin=<-15 -15 -15>, angularTarget=<1 0 0>}
      localJointOffset=<2 0 0>
     }]
   }
   n38 = Anim Dynamics - Bone: dyn_pole_01(ComponentPose=@n37, Alpha=0.5, ExternalForce="0,0,0") {
    alpha=0.2
    angularSpringConstant=10000
    bAngularSpring=true
    linearSpringConstant=1000
    simulationSpace=RootRelative
    boundBone {boneName=dyn_pole_01}
    gravityOverride {z=-1500}
    physicsBodyDefinitions [{
      boundBone {boneName=dyn_pole_01}
      boxExtents=<5 5 5>
      constraintSetup {angularLimitsMax=<0 0 5>, angularLimitsMin=<0 0 -5>, angularTarget=<0 0 0>}
      localJointOffset=<0.1 0 0>
     }]
    simSpaceSettings {simSpaceAngularAlpha=0.4}
   }
   n39 = Anim Dynamics - Bone: dyn_pole_02(ComponentPose=@n38, Alpha=0.5, ExternalForce="0,0,0") {
    alpha=0.2
    angularSpringConstant=10000
    bAngularSpring=true
    linearSpringConstant=1000
    simulationSpace=RootRelative
    boundBone {boneName=dyn_pole_02}
    gravityOverride {z=-1500}
    physicsBodyDefinitions [{
      boundBone {boneName=dyn_pole_02}
      boxExtents=<5 5 5>
      constraintSetup {angularLimitsMax=<0 0 5>, angularLimitsMin=<0 0 -5>, angularTarget=<0 0 0>}
      localJointOffset=<0.1 0 0>
     }]
    simSpaceSettings {simSpaceAngularAlpha=0.4}
   }
   n40 = Anim Dynamics - Bone: dyn_pole_03(ComponentPose=@n39, Alpha=0.5, ExternalForce="0,0,0") {
    alpha=0.2
    angularSpringConstant=10000
    bAngularSpring=true
    linearSpringConstant=1000
    simulationSpace=RootRelative
    boundBone {boneName=dyn_pole_03}
    gravityOverride {z=-1500}
    physicsBodyDefinitions [{
      boundBone {boneName=dyn_pole_03}
      boxExtents=<5 5 5>
      constraintSetup {angularLimitsMax=<0 0 5>, angularLimitsMin=<0 0 -5>, angularTarget=<0 0 0>}
      localJointOffset=<0.1 0 0>
     }]
    simSpaceSettings {simSpaceAngularAlpha=0.4}
   }
   n41 = Anim Dynamics - Bone: dyn_pole_04(ComponentPose=@n40, Alpha=0.2, ExternalForce="0,0,0") {
    alpha=0.2
    angularSpringConstant=10000
    bAngularSpring=true
    linearSpringConstant=1000
    simulationSpace=RootRelative
    boundBone {boneName=dyn_pole_04}
    gravityOverride {z=-1500}
    physicsBodyDefinitions [{
      boundBone {boneName=dyn_pole_04}
      boxExtents=<5 5 5>
      constraintSetup {angularLimitsMax=<0 0 5>, angularLimitsMin=<0 0 -5>, angularTarget=<0 0 0>}
      localJointOffset=<0.1 0 0>
     }]
    simSpaceSettings {simSpaceAngularAlpha=0.4}
   }
   n42 = Anim Dynamics - Bone: dyn_pole_05(ComponentPose=@n41, Alpha=0.2, ExternalForce="0,0,0") {
    alpha=0.2
    angularSpringConstant=10000
    bAngularSpring=true
    linearSpringConstant=1000
    simulationSpace=RootRelative
    boundBone {boneName=dyn_pole_05}
    gravityOverride {z=-1500}
    physicsBodyDefinitions [{
      boundBone {boneName=dyn_pole_05}
      boxExtents=<5 5 5>
      constraintSetup {angularLimitsMax=<0 0 5>, angularLimitsMin=<0 0 -5>, angularTarget=<0 0 0>}
      localJointOffset=<0.1 0 0>
     }]
    simSpaceSettings {simSpaceAngularAlpha=0.4}
   }
   n43 = Anim Dynamics - Bone: dyn_pole_06(ComponentPose=@n42, Alpha=0.2, ExternalForce="0,0,0") {
    alpha=0.2
    angularSpringConstant=10000
    bAngularSpring=true
    linearSpringConstant=1000
    simulationSpace=RootRelative
    boundBone {boneName=dyn_pole_06}
    gravityOverride {z=-1500}
    physicsBodyDefinitions [{
      boundBone {boneName=dyn_pole_06}
      boxExtents=<5 5 5>
      constraintSetup {angularLimitsMax=<0 0 5>, angularLimitsMin=<0 0 -5>, angularTarget=<0 0 0>}
      localJointOffset=<0.1 0 0>
     }]
    simSpaceSettings {simSpaceAngularAlpha=0.4}
   }
   n44 = Anim Dynamics - Bone: dyn_pole_07(ComponentPose=@n43, Alpha=0.2, ExternalForce="0,0,0") {
    alpha=0.2
    angularSpringConstant=10000
    bAngularSpring=true
    linearSpringConstant=1000
    simulationSpace=RootRelative
    boundBone {boneName=dyn_pole_07}
    gravityOverride {z=-1500}
    physicsBodyDefinitions [{
      boundBone {boneName=dyn_pole_07}
      boxExtents=<5 5 5>
      constraintSetup {angularLimitsMax=<0 0 5>, angularLimitsMin=<0 0 -5>, angularTarget=<0 0 0>}
      localJointOffset=<0.1 0 0>
     }]
    simSpaceSettings {simSpaceAngularAlpha=0.4}
   }
   n45 = "Transform (Modify) Bone - Bone: dyn_pole_01"(ComponentPose=@n44, bAlphaBoolEnabled=True, Translation="0,0,0", Scale="1,1,1") {
    alphaInputType=Bool
    rotationMode=BMM_Additive
    rotationSpace=BCS_BoneSpace
    translationMode=BMM_Additive
    translationSpace=BCS_BoneSpace
    alphaBoolBlend {blendInTime=0.3, blendOutTime=0.3}
    boneToModify {boneName=dyn_pole_01}
    rotation {yaw=-6}
   }
   n46 = "Transform (Modify) Bone - Bone: dyn_pole_02"(ComponentPose=@n45, bAlphaBoolEnabled=True, Translation="0,0,0", Scale="1,1,1") {
    alphaInputType=Bool
    rotationMode=BMM_Additive
    rotationSpace=BCS_BoneSpace
    translationMode=BMM_Additive
    translationSpace=BCS_BoneSpace
    alphaBoolBlend {blendInTime=0.3, blendOutTime=0.3}
    boneToModify {boneName=dyn_pole_02}
    rotation {yaw=-6}
   }
   n47 = "Transform (Modify) Bone - Bone: dyn_pole_03"(ComponentPose=@n46, bAlphaBoolEnabled=True, Translation="0,0,0", Scale="1,1,1") {
    alphaInputType=Bool
    rotationMode=BMM_Additive
    rotationSpace=BCS_BoneSpace
    translationMode=BMM_Additive
    translationSpace=BCS_BoneSpace
    alphaBoolBlend {blendInTime=0.3, blendOutTime=0.3}
    boneToModify {boneName=dyn_pole_03}
    rotation {yaw=-6}
   }
   n48 = "Transform (Modify) Bone - Bone: dyn_pole_04"(ComponentPose=@n47, bAlphaBoolEnabled=True, Translation="0,0,0", Scale="1,1,1") {
    alphaInputType=Bool
    rotationMode=BMM_Additive
    rotationSpace=BCS_BoneSpace
    translationMode=BMM_Additive
    translationSpace=BCS_BoneSpace
    alphaBoolBlend {blendInTime=0.3, blendOutTime=0.3}
    boneToModify {boneName=dyn_pole_04}
    rotation {yaw=-6}
   }
   n49 = "Transform (Modify) Bone - Bone: dyn_pole_05"(ComponentPose=@n48, bAlphaBoolEnabled=True, Translation="0,0,0", Scale="1,1,1") {
    alphaInputType=Bool
    rotationMode=BMM_Additive
    rotationSpace=BCS_BoneSpace
    translationMode=BMM_Additive
    translationSpace=BCS_BoneSpace
    alphaBoolBlend {blendInTime=0.3, blendOutTime=0.3}
    boneToModify {boneName=dyn_pole_05}
    rotation {yaw=-4}
   }
   n50 = "Transform (Modify) Bone - Bone: dyn_pole_06"(ComponentPose=@n49, bAlphaBoolEnabled=True, Translation="0,0,0", Scale="1,1,1") {
    alphaInputType=Bool
    rotationMode=BMM_Additive
    rotationSpace=BCS_BoneSpace
    translationMode=BMM_Additive
    translationSpace=BCS_BoneSpace
    alphaBoolBlend {blendInTime=0.3, blendOutTime=0.3}
    boneToModify {boneName=dyn_pole_06}
    rotation {yaw=-4}
   }
   n51 = "Transform (Modify) Bone - Bone: dyn_pole_07"(ComponentPose=@n50, bAlphaBoolEnabled=True, Translation="0,0,0", Scale="1,1,1") {
    alphaInputType=Bool
    rotationMode=BMM_Additive
    rotationSpace=BCS_BoneSpace
    translationMode=BMM_Additive
    translationSpace=BCS_BoneSpace
    alphaBoolBlend {blendInTime=0.3, blendOutTime=0.3}
    boneToModify {boneName=dyn_pole_07}
    rotation {yaw=-8}
   }
   n52 = Component To Local(ComponentPose=@n51)
   n53 = RigidBody!(ComponentPose=Local To Component(LocalPose=Use cached pose 'GrassCache_1'()), Alpha:float, OverrideWorldGravity="0,0,0", ExternalForce) -> (Pose) {
    bClampLinearTranslationLimitToRefPose=true
    bOverrideWorldGravity=true
    overridePhysicsAsset=&ChefTartar_Axe_Physics
    simulationSpace=BaseBoneSpace
    alphaScaleBiasClamp {bInterpResult=true, interpSpeedDecreasing=5, interpSpeedIncreasing=15}
    baseBoneRef {boneName=b_AR_Root}
    componentAppliedLinearAccClamp=<0 0 0>
    simSpaceSettings {
     maxAngularAcceleration=20
     maxAngularVelocity=2
     maxLinearAcceleration=700
     maxLinearVelocity=500
     worldAlpha=1
     externalLinearDragV=<0.2 0.2 0.2>
    }
   }
   n54 = Anim Dynamics - Bone: dyn_flag_01!(ComponentPose=@n53, ExternalForce="0,0,0") -> (Pose) {
    angularDampingOverride=1
    angularSpringConstant=20000
    bOverrideAngularDamping=true
    bUseGravityOverride=true
    gravityScale=0.35
    simulationSpace=RootRelative
    boundBone {boneName=dyn_flag_01}
    componentLinearAccScale=<1 1 1>
    componentLinearVelScale=<1 1 1>
    gravityOverride {y=-980}
    physicsBodyDefinitions [{
      boundBone {boneName=dyn_flag_01}
      boxExtents=<20 30 30>
      constraintSetup {angularLimitsMax=<0 0 7>, angularLimitsMin=<0 0 -7>, angularTarget=<1 0 0>}
      localJointOffset=<4 0 0>
     }]
   }
   n55 = Anim Dynamics - Bone: dyn_flag_02!(ComponentPose=@n54, ExternalForce="0,0,0") -> (Pose) {
    angularDampingOverride=1
    angularSpringConstant=20000
    bOverrideAngularDamping=true
    bUseGravityOverride=true
    gravityScale=0.35
    simulationSpace=RootRelative
    boundBone {boneName=dyn_flag_02}
    componentLinearAccScale=<1 1 1>
    componentLinearVelScale=<1 1 1>
    gravityOverride {y=-980}
    physicsBodyDefinitions [{
      boundBone {boneName=dyn_flag_02}
      boxExtents=<20 30 30>
      constraintSetup {angularLimitsMax=<0 0 7>, angularLimitsMin=<0 0 -7>, angularTarget=<1 0 0>}
      localJointOffset=<4 0 0>
     }]
   }
   n56 = Anim Dynamics - Bone: dyn_flag_03!(ComponentPose=@n55, ExternalForce="0,0,0") -> (Pose) {
    angularDampingOverride=1
    angularSpringConstant=20000
    bOverrideAngularDamping=true
    bUseGravityOverride=true
    gravityScale=0.35
    simulationSpace=RootRelative
    boundBone {boneName=dyn_flag_03}
    componentLinearAccScale=<1 1 1>
    componentLinearVelScale=<1 1 1>
    gravityOverride {y=-980}
    physicsBodyDefinitions [{
      boundBone {boneName=dyn_flag_03}
      boxExtents=<20 30 30>
      constraintSetup {angularLimitsMax=<0 0 7>, angularLimitsMin=<0 0 -7>, angularTarget=<1 0 0>}
      localJointOffset=<4 0 0>
     }]
   }
   n57 = Anim Dynamics - Bone: dyn_flag_04!(ComponentPose=@n56, ExternalForce="0,0,0") -> (Pose) {
    angularDampingOverride=1
    angularSpringConstant=20000
    bOverrideAngularDamping=true
    bUseGravityOverride=true
    gravityScale=0.35
    simulationSpace=RootRelative
    boundBone {boneName=dyn_flag_04}
    componentLinearAccScale=<1 1 1>
    componentLinearVelScale=<1 1 1>
    gravityOverride {y=-980}
    physicsBodyDefinitions [{
      boundBone {boneName=dyn_flag_04}
      boxExtents=<20 30 30>
      constraintSetup {angularLimitsMax=<0 0 7>, angularLimitsMin=<0 0 -7>, angularTarget=<1 0 0>}
      localJointOffset=<4 0 0>
     }]
   }
   n58 = Anim Dynamics - Bone: dyn_flag_05!(ComponentPose=@n57, ExternalForce="0,0,0") -> (Pose) {
    angularDampingOverride=1
    angularSpringConstant=20000
    bOverrideAngularDamping=true
    bUseGravityOverride=true
    gravityScale=0.35
    simulationSpace=RootRelative
    boundBone {boneName=dyn_flag_05}
    componentLinearAccScale=<1 1 1>
    componentLinearVelScale=<1 1 1>
    gravityOverride {y=-980}
    physicsBodyDefinitions [{
      boundBone {boneName=dyn_flag_05}
      boxExtents=<20 30 30>
      constraintSetup {angularLimitsMax=<0 0 7>, angularLimitsMin=<0 0 -7>, angularTarget=<1 0 0>}
      localJointOffset=<4 0 0>
     }]
   }
   n59 = Anim Dynamics - Bone: dyn_flag_06!(ComponentPose=@n58, ExternalForce="0,0,0") -> (Pose) {
    angularDampingOverride=1
    angularSpringConstant=20000
    bOverrideAngularDamping=true
    bUseGravityOverride=true
    gravityScale=0.35
    simulationSpace=RootRelative
    boundBone {boneName=dyn_flag_06}
    componentLinearAccScale=<1 1 1>
    componentLinearVelScale=<1 1 1>
    gravityOverride {y=-980}
    physicsBodyDefinitions [{
      boundBone {boneName=dyn_flag_06}
      boxExtents=<20 30 30>
      constraintSetup {angularLimitsMax=<0 0 7>, angularLimitsMin=<0 0 -7>, angularTarget=<1 0 0>}
      localJointOffset=<4 0 0>
     }]
   }
   Output Pose(Result=Blend Poses by bool(BlendPose_0=Blend Poses by bool(BlendPose_0=@n29, BlendPose_1=@n52), BlendPose_1=Component To Local(ComponentPose=@n59))) {layerGroup=None}
   n60 = Control Rig!(Source=Input Pose(), Alpha, is in FRONT END:bool=Get bIsInFrontEnd(), is on PAWN:bool=Get bIsOnPawn() -> (bIsOnPawn:bool)) -> (Pose) {controlRigAssetReference {blueprintRigClass=&/ChefTartar_Axe_CtrlRig.ChefTartar_Axe_CtrlRig_C}}
   Save cached pose 'input_pose_cache'(Pose=@n60) {cacheName=input_pose_cache}
   n61 = Anim Dynamics - Bone: base_grass_01!(ComponentPose=Local To Component(LocalPose=Use cached pose 'input_pose_cache'()), Alpha:float=1, ExternalForce="0,0,0") -> (Pose) {
    angularDampingOverride=1
    angularSpringConstant=20000
    bAngularSpring=true
    bOverrideAngularDamping=true
    bUseGravityOverride=true
    gravityScale=0.35
    simulationSpace=RootRelative
    boundBone {boneName=base_grass_01}
    componentLinearAccScale=<1 1 1>
    componentLinearVelScale=<1 1 1>
    gravityOverride {z=-980}
    physicsBodyDefinitions [{
      boundBone {boneName=base_grass_01}
      boxExtents=<20 30 30>
      constraintSetup {angularLimitsMax=<15 15 15>, angularLimitsMin=<-15 -15 -15>, angularTarget=<1 0 0>}
      localJointOffset=<4 0 0>
     }]
   }
   n62 = Anim Dynamics - Bone: dyn_grass_02!(ComponentPose=@n61, Alpha:float=1, ExternalForce="0,0,0") -> (Pose) {
    angularDampingOverride=1
    angularSpringConstant=20000
    bAngularSpring=true
    bOverrideAngularDamping=true
    bUseGravityOverride=true
    gravityScale=0.35
    simulationSpace=RootRelative
    boundBone {boneName=dyn_grass_02}
    componentLinearAccScale=<1 1 1>
    componentLinearVelScale=<1 1 1>
    gravityOverride {z=-980}
    physicsBodyDefinitions [{
      boundBone {boneName=dyn_grass_02}
      boxExtents=<20 30 30>
      constraintSetup {angularLimitsMax=<15 15 15>, angularLimitsMin=<-15 -15 -15>, angularTarget=<1 0 0>}
      localJointOffset=<4 0 0>
     }]
   }
   n63 = Anim Dynamics - Bone: base_grass_03!(ComponentPose=@n62, Alpha:float=1, ExternalForce="0,0,0") -> (Pose) {
    angularDampingOverride=1
    angularSpringConstant=20000
    bAngularSpring=true
    bOverrideAngularDamping=true
    bUseGravityOverride=true
    gravityScale=0.35
    simulationSpace=RootRelative
    boundBone {boneName=base_grass_03}
    componentLinearAccScale=<1 1 1>
    componentLinearVelScale=<1 1 1>
    gravityOverride {z=-980}
    physicsBodyDefinitions [{
      boundBone {boneName=base_grass_03}
      boxExtents=<20 30 30>
      constraintSetup {angularLimitsMax=<15 15 15>, angularLimitsMin=<-15 -15 -15>, angularTarget=<1 0 0>}
      localJointOffset=<4 0 0>
     }]
   }
   n64 = Anim Dynamics - Bone: dyn_grass_04!(ComponentPose=@n63, Alpha:float=1, ExternalForce="0,0,0") -> (Pose) {
    angularDampingOverride=1
    angularSpringConstant=20000
    bAngularSpring=true
    bOverrideAngularDamping=true
    bUseGravityOverride=true
    gravityScale=0.35
    simulationSpace=RootRelative
    boundBone {boneName=dyn_grass_04}
    componentLinearAccScale=<1 1 1>
    componentLinearVelScale=<1 1 1>
    gravityOverride {z=-980}
    physicsBodyDefinitions [{
      boundBone {boneName=dyn_grass_04}
      boxExtents=<20 30 30>
      constraintSetup {angularLimitsMax=<15 15 15>, angularLimitsMin=<-15 -15 -15>, angularTarget=<1 0 0>}
      localJointOffset=<4 0 0>
     }]
   }
   n65 = Anim Dynamics - Bone: base_grass_05!(ComponentPose=@n64, Alpha:float=1, ExternalForce="0,0,0") -> (Pose) {
    angularDampingOverride=1
    angularSpringConstant=20000
    bAngularSpring=true
    bOverrideAngularDamping=true
    bUseGravityOverride=true
    gravityScale=0.35
    simulationSpace=RootRelative
    boundBone {boneName=base_grass_05}
    componentLinearAccScale=<1 1 1>
    componentLinearVelScale=<1 1 1>
    gravityOverride {z=-980}
    physicsBodyDefinitions [{
      boundBone {boneName=base_grass_05}
      boxExtents=<20 30 30>
      constraintSetup {angularLimitsMax=<15 15 15>, angularLimitsMin=<-15 -15 -15>, angularTarget=<1 0 0>}
      localJointOffset=<4 0 0>
     }]
   }
   n66 = Anim Dynamics - Bone: dyn_grass_06!(ComponentPose=@n65, Alpha:float=1, ExternalForce="0,0,0") -> (Pose) {
    angularDampingOverride=1
    angularSpringConstant=20000
    bAngularSpring=true
    bOverrideAngularDamping=true
    bUseGravityOverride=true
    gravityScale=0.35
    simulationSpace=RootRelative
    boundBone {boneName=dyn_grass_06}
    componentLinearAccScale=<1 1 1>
    componentLinearVelScale=<1 1 1>
    gravityOverride {z=-980}
    physicsBodyDefinitions [{
      boundBone {boneName=dyn_grass_06}
      boxExtents=<20 30 30>
      constraintSetup {angularLimitsMax=<15 15 15>, angularLimitsMin=<-15 -15 -15>, angularTarget=<1 0 0>}
      localJointOffset=<4 0 0>
     }]
   }
   n67 = Anim Dynamics - Bone: base_grass_07!(ComponentPose=@n66, Alpha:float=1, ExternalForce="0,0,0") -> (Pose) {
    angularDampingOverride=1
    angularSpringConstant=20000
    bAngularSpring=true
    bOverrideAngularDamping=true
    bUseGravityOverride=true
    gravityScale=0.35
    simulationSpace=RootRelative
    boundBone {boneName=base_grass_07}
    componentLinearAccScale=<1 1 1>
    componentLinearVelScale=<1 1 1>
    gravityOverride {z=-980}
    physicsBodyDefinitions [{
      boundBone {boneName=base_grass_07}
      boxExtents=<20 30 30>
      constraintSetup {angularLimitsMax=<15 15 15>, angularLimitsMin=<-15 -15 -15>, angularTarget=<1 0 0>}
      localJointOffset=<4 0 0>
     }]
   }
   n68 = Anim Dynamics - Bone: base_grass_08!(ComponentPose=@n67, Alpha:float=1, ExternalForce="0,0,0") -> (Pose) {
    angularDampingOverride=1
    angularSpringConstant=20000
    bAngularSpring=true
    bOverrideAngularDamping=true
    bUseGravityOverride=true
    gravityScale=0.35
    simulationSpace=RootRelative
    boundBone {boneName=base_grass_08}
    componentLinearAccScale=<1 1 1>
    componentLinearVelScale=<1 1 1>
    gravityOverride {z=-980}
    physicsBodyDefinitions [{
      boundBone {boneName=base_grass_08}
      boxExtents=<20 30 30>
      constraintSetup {angularLimitsMax=<15 15 15>, angularLimitsMin=<-15 -15 -15>, angularTarget=<1 0 0>}
      localJointOffset=<4 0 0>
     }]
   }
   Save cached pose 'GrassCache_1'(Pose=Component To Local(ComponentPose=@n68)) {cacheName=GrassCache_1}
  ]}
 EventGraph {nodes [
   Event BlueprintUpdateAnimation()
   > Set Dynamics_Alpha(Dynamics_Alpha:double=SelectFloat(A=1, B=0.5, bPickA=Get bIsPlayingMeleeAnim())) -> (Output_Get:double)
   TryGetPawnOwner()
  ]}
]
variables [Dynamics_Alpha {type=number}]
